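# Sound cue for the looping ambient spot sound of a Helios water fountain environment actor.
# A single looping wave player feeds a medium stream recording straight to the output, boosted in volume and routed to the outdoor Athena ambient sound class.
# A fountain-specific sphere attenuation asset supplies spatialization with listener focus, while a concurrency override caps simultaneous instances.
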SoundCue /Game/Sounds/Helios/Environment_Actors/WaterFountain/Spot_Water_Fountain_Loop_Cue
root=WavePlayer_0
nodes [{id=WavePlayer_0, properties {bLooping=true, soundWaveAssetPtr=/Game/Sounds/Fort_Ambient_Looping/Water/stream_medium_B_01}}]
properties {
 attenuationSettings=&/Fort_Water_Fountain_Focus_Sphere_Att
 bOverrideConcurrency=true
 maxDistance=1250
 soundClassObject=&Fort_Ambient_Outdoors_Athena
 volumeMultiplier=4
 attenuationOverrides {
  bEnableFocusInterpolation=true
  bEnableListenerFocus=true
  falloffDistance=2000
  focusAzimuth=25
  nonFocusAzimuth=45
  nonFocusVolumeAttenuation=0.5
  nonSpatializedRadiusStart=300
  stereoSpread=1000
  attenuationShapeExtents {x=200}
 }
 concurrencyOverrides {maxCount=2}
}
