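# Sound cue for a small water fountain ambient loop, played in the Athena outdoor ambient sound class.
# One looping medium stream wave runs through a mixer and a quality-level node, which feeds that same mixer at every level, into a modulator adding slight upward random pitch variation.
# A small-fountain focus-sphere attenuation asset handles spatial falloff, while overridden concurrency limits the cue to two simultaneous instances at reduced volume.
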SoundCue /Game/Sounds/Helios/Environment_Actors/WaterFountain/Spot_Water_Fountain_Loop_Small_Cue
root=Modulator_1
nodes [
 {id=Mixer_0, children [WavePlayer_2,null], properties {inputVolume [1,1]}}
 {id=Modulator_1, children [QualityLevel_0], properties {pitchMax=1.1, pitchMin=1, volumeMax=1, volumeMin=1}}
 {id=QualityLevel_0, children [Mixer_0,Mixer_0,Mixer_0,Mixer_0]}
 {id=WavePlayer_2, properties {bLooping=true, soundWaveAssetPtr=/Game/Sounds/Fort_Ambient_Looping/Water/stream_medium_B_01_opt}}
]
properties {
 attenuationSettings=&/Fort_Water_Fountain_Focus_Sphere_Small_Att
 bOverrideConcurrency=true
 maxDistance=750
 soundClassObject=&Fort_Ambient_Outdoors_Athena
 volumeMultiplier=0.401
 attenuationOverrides {
  bEnableFocusInterpolation=true
  bEnableListenerFocus=true
  falloffDistance=2000
  focusAzimuth=25
  nonFocusAzimuth=45
  nonFocusVolumeAttenuation=0.5
  nonSpatializedRadiusStart=300
  stereoSpread=1000
  attenuationShapeExtents {x=200}
 }
 concurrencyOverrides {bVolumeScaleCanRelease=true, maxCount=2}
}
